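# Surface material for Asteria gravel in Creative landscapes, usable on instanced, Nanite and spline meshes.
# Vertex color channels, optionally hue-shifted per instance, blend tint colors, with macro variation and overlay blending.
# World-aligned detail textures and cell noise supply masks and normals, while a static switch adds regular texture maps with per-channel influence.
# A world-height ocean mask drives a wetness blend that scales albedo and roughness near water level.
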Material /BRCosmetics/Creative/Materials/Landscape/Creative_Landscape_Asteria/Grass/Materials/M_CR_AsteriaGravel
graph {
 properties {
  bTangentSpaceNormal=false
  bUseMaterialAttributes=true
  bUsedWithInstancedStaticMeshes=true
  bUsedWithNanite=true
  bUsedWithSplineMeshes=true
  pixelDepthOffsetMode=PDOM_Legacy
  usedShadingModels=Default Lit
 }
 nodes [
  Constant() {r=0.0002}
  Constant() {r=4}
  Constant() {r=7}
  Constant() {r=1}
  n1 = TextureObject() {samplerType=SAMPLERTYPE_Grayscale, texture=/BRCosmetics/Weapons/FORT_Melee/Pickaxe_Cube_Ninja_Male/FX/T_CellNoise}
  MaterialFunctionCall!(
   Texture=@n1
   LandscapeCoordinates=Multiply(A=ScalarParameter() {defaultValue=0.004, group=Variation, parameterName=GravelVariationMask_02_Tiling}, B=WorldPosition().XYZ)
   LowEndTiling
   HexGridGradientSharpness
   HexGridSize
   HexGridUVInput
   MinScale
   MaxScale
   MinOffset
   MaxOffset
   "MaxRotation(Degrees)"
   "MinRotation(Degrees)"
   forceLowQuality
  ) {materialFunction=/Game/Athena/Asteria/Environments/Landscape/Materials/MaterialFunctions/UtilityFunctions/Landscape_TextureSample}
  n2 = MaterialFunctionCall!(
   Diffuse Texture Object=TextureObject() {samplerType=SAMPLERTYPE_Masks, texture=/Game/Athena/WinterArt/Textures/T_DirtLeaks_M}
   Spec Texture Object
   TextureSize=ScalarParameter() {defaultValue=64, group=DetailNoise, parameterName=DetailNoiseScale}
   --------------------------
   Use seperate settings for Z projection
   Use seperate texture for Z projection
   Texture Size Z projection
   Z Plane Diffuse Texture Object
   Z Plane Spec Texture Object
   "--------------------------      "
   ProjectionTransitionContrast
   Use Texture Alpha Heightmap to Refine the Transistion
   HeightLerpTransitionContrast
   PickTheDiffuseHeightChannel
   TheDiffuseHeightChannel
   "--------------------------       "
   NormalTextureObject=TextureObject() {samplerType=SAMPLERTYPE_Normal, texture=/Game/Environments/Apollo/Rocks/Natural_Rocks/Textures/T_RockDetails_02_N}
   World Space Normals=StaticBool() {value=true}
   "--------------------------  "
   Export Float 4 Diffuse
   Export Float 4 Spec
   "--------------------------     "
   WorldPosition
   World Space Normal
  ) {materialFunction=/Engine/Functions/Engine_MaterialFunctions02/Texturing/WorldAlignedTextures_Complex}
  Multiply(A=@n2.Normal Z Texture, B=VectorParameter() {parameterName=DetailNormalStrength, defaultValue=<rgba 2 2 0 1>}.RGB)
  NamedRerouteDeclaration(None=TextureSampleParameter2D() {group=RegularTextures, parameterName=T_BaseColor}.RGB) {name=Texture_BaseColor, nodeColor=<rgba 0 1 0.1562 1>}
  n3 = MaterialFunctionCall(Float3=@n2.Z Diffuse Texture) {materialFunction=&BreakOutFloat3Components}
  NamedRerouteDeclaration(None=@n3.B) {name=Details-B, nodeColor=<rgba 0.4062 1 0 1>}
  NamedRerouteDeclaration(None=TextureSampleParameter2D() {group=RegularTextures, parameterName=T_Normal, samplerType=SAMPLERTYPE_Normal, texture=/Engine/EngineMaterials/FlatNormal}.RGB) {name=Texture_Normal, nodeColor=<rgba 1 0.7266 0 1>}
  n4 = WorldPosition() {worldPositionShaderOffset=WPT_ExcludeAllShaderOffsets}
  n5 = Add(
   A=Add(A=ComponentMask(None=@n4.XYZ) {b=true}, B=ScalarParameter() {defaultValue=64, parameterName=WaterLevelOffset})
   B=Constant() {desc=slightly wet sand height offset, r=10}
  )
  n6 = ScalarParameter() {defaultValue=90, parameterName=WaterLevelTransition}
  NamedRerouteDeclaration(None=Saturate(None=Divide(A=@n5, B=@n6))) {name=OceanMask, nodeColor=<rgba 0.5 0 1 1>}
  NamedRerouteDeclaration(None=@n2.Normal Z Texture) {name=Details N, nodeColor=<rgba 0 1 0.5078 1>}
  n7 = ScalarParameter() {defaultValue=300, group=Variation, parameterName=VariationTextureScale}
  n8 = Multiply(
   A=ComponentMask(None=MaterialFunctionCall(TextureObject=@n1, TextureSize=@n7) {materialFunction=/Game/Weapons/WeaponSkins/Wraps/MaterialFunctions/WorldAlignedTexture}.Z Texture) {r=true}
   B=ScalarParameter() {defaultValue=5, group=Variation, parameterName=VariationTextureMod}
  )
  n9 = ScalarParameter() {defaultValue=1, group=Variation, parameterName=VariationTexturePow}
  NamedRerouteDeclaration(None=Clamp(None=Power(Base=@n8, Exp=@n9))) {name=VariationMask_02, nodeColor=<rgba 1 0 0.6562 1>}
  n10 = TextureSampleParameter2D() {group=RegularTextures, parameterName=T_Specular, samplerType=SAMPLERTYPE_Masks, texture=/Engine/EngineMaterials/DefaultDiffuse_TC_Masks}
  NamedRerouteDeclaration(None=@n10.R) {name=Texture_Specular, nodeColor=<rgba 0 0.8359 1 1>}
  NamedRerouteDeclaration(None=@n3.R) {name=Details-R, nodeColor=<rgba 0 1 0.5078 1>}
  NamedRerouteDeclaration(None=@n10.B) {name=Texture_Roughness, nodeColor=<rgba 1 0.375 0 1>}
  NamedRerouteDeclaration(None=@n3.G) {name=Details-G, nodeColor=<rgba 0.6172 0 1 1>}
  n11 = GetMaterialAttributes() {attributeGetTypes [#1]}
  SetMaterialAttributes!(
   MaterialAttributes=@n11.MaterialAttributes
   Base Color=LinearInterpolate(A=@n11.Base Color, B=VertexColor(), Alpha=ScalarParameter() {parameterName=debugvertcolor})
  ) {attributeSetTypes [#1]}
  n12 = ScalarParameter() {defaultValue=10, group=Colors, parameterName=BaseMacroHueValue, sortPriority=-125}
  n13 = VectorParameter() {group=Colors, parameterName=Base-MacroHueShift, sortPriority=-150, defaultValue=<rgba 0.0704 0.0592 0.0513 1>}
  n14 = VectorParameter() {group=Colors, parameterName=BaseColor, sortPriority=-500, defaultValue=<rgba 0.0434 0.04 0.0303 0>}
  n15 = VectorParameter() {group=Colors, parameterName=BaseColor_VertexRed, sortPriority=-400, defaultValue=<rgba 0.0605 0.0747 0.0691 0>}
  n16 = VertexColor()
  n17 = PerInstanceRandom()
  n18 = MaterialFunctionCall(Float3=LinearInterpolate(
    A=@n16
    B=MaterialFunctionCall(Hue Shift Percentage=@n17, Texture=@n16) {materialFunction=&HueShift}.Result
    Alpha=ScalarParameter() {group=Colors, parameterName=UseVertexColorMaskHueShift, sortPriority=-100}
   )) {materialFunction=&BreakOutFloat3Components}
  n19 = LinearInterpolate(A=@n14.RGB, B=@n15.RGB, Alpha=@n18.R)
  n20 = VectorParameter() {group=Colors, parameterName=BaseColor_VertexGreen, sortPriority=-300, defaultValue=<rgba 0.1372 0.0931 0.0593 0>}
  n21 = VectorParameter() {group=Colors, parameterName=BaseColor_VertexBlue, sortPriority=-200, defaultValue=<rgba 0.0677 0.0557 0.0469 0>}
  n22 = LinearInterpolate(
   A=@n14.RGB
   B=LinearInterpolate(A=LinearInterpolate(A=@n19, B=@n20.RGB, Alpha=@n18.G), B=@n21.RGB, Alpha=@n18.B)
   Alpha=ScalarParameter() {defaultValue=1, group=Colors, parameterName=UseVertexColorMask, sortPriority=-100}
  )
  n23 = Multiply(A=@n13.RGB, B=@n22)
  n24 = MaterialFunctionCall(Base=LinearInterpolate(A=Multiply(A=@n12, B=@n23), B=@n22, Alpha=NamedRerouteUsage()), Blend=NamedRerouteUsage()) {materialFunction=&Blend_Overlay}
  n25 = Saturate(None=LinearInterpolate(
    A=ScalarParameter() {defaultValue=0.25, group=DetailNoise, parameterName=SpecMin}
    B=ScalarParameter() {defaultValue=1, group=DetailNoise, parameterName=SpecMax}
    Alpha=NamedRerouteUsage()
   ))
  n26 = Saturate(None=LinearInterpolate(
    A=ScalarParameter() {defaultValue=0.1, group=DetailNoise, parameterName=RoughnessMin}
    B=ScalarParameter() {defaultValue=1, group=DetailNoise, parameterName=RoughnessMax}
    Alpha=NamedRerouteUsage()
   ))
  n27 = NamedRerouteUsage()
  n28 = VectorParameter() {group=RegularTextures, parameterName=TextureNormalAmount, defaultValue=<rgba 1 1 1 1>}
  n29 = SetMaterialAttributes!(
   MaterialAttributes
   Base Color=LinearInterpolate(
    A=@n24.Result
    B=NamedRerouteUsage()
    Alpha=ScalarParameter() {defaultValue=1, group=RegularTextures, parameterName=BaseColorTextureInfluence}
   )
   Specular=LinearInterpolate(A=@n25, B=NamedRerouteUsage(), Alpha=ScalarParameter() {defaultValue=1, group=RegularTextures, parameterName=SpecTextureInfluence})
   Roughness=LinearInterpolate(A=@n26, B=NamedRerouteUsage(), Alpha=ScalarParameter() {defaultValue=1, group=RegularTextures, parameterName=RoughTextureInfluence})
   Normal=Transform(None=Multiply(A=@n27, B=@n28.RGB))
  ) {attributeSetTypes [
    #1
    #2
    #3
    #4
   ]}
  n30 = SetMaterialAttributes(Base Color=@n24.Result, Specular=@n25, Roughness=@n26, Normal=NamedRerouteUsage()) {attributeSetTypes [
    #1
    #2
    #3
    #4
   ]}
  n31 = StaticSwitchParameter(True=@n29, False=@n30) {group=RegularTextures, parameterName=UseRegularTextures}
  n32 = GetMaterialAttributes(None=@n31) {attributeGetTypes [#1, #2, #3]}
  n33 = Multiply(A=@n32.Base Color, B=ScalarParameter() {defaultValue=1, parameterName=Wet_AlbedoDarken})
  n34 = Multiply(A=@n32.Roughness, B=ScalarParameter() {defaultValue=0.5, parameterName=Wet_RoughnessMult})
  n35 = GetMaterialAttributes(None=BlendMaterialAttributes(
    A=SetMaterialAttributes!(MaterialAttributes=@n32.MaterialAttributes, Base Color=@n33, Specular=@n32.MaterialAttributes, Roughness=@n34) {attributeSetTypes [#1, #2, #3]}
    B=@n31
    Alpha=NamedRerouteUsage()
   )) {attributeGetTypes [#4]}
  SetMaterialAttributes!(MaterialAttributes=@n35.MaterialAttributes, Normal=Transform(None=@n35.Normal)) {attributeSetTypes [#4]}
 ]
}
